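; Victory Crown flare gameplay cue that shows a crown-icon sound indicator at the target actor when the burst fires.
; On burst, it calls the local team sound indicator with the target's location, team and controller, covering hostile, neutral and friendly affiliations.
; The indicator uses the Chest type, with its range taken from the first burst sound's max distance, or zero if none exists.
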(fn GetSoundIndicatorMaxDistance
  (bind _bursteffectdata (Variables|BurstEffects|GetBurstEffectData))
  (bind (_burstparticles _burstsounds) (Utilities|Struct|BreakFortBurstEffectData _bursteffectdata))
  (bind (_condition _attachment _soundcue) (Utilities|Struct|BreakFortGameplayCueAudioInfo (Utilities|Array|Get(aref) _burstsounds)))
  (if (Utilities|Array|IsValidIndex _burstsounds)
    (return (Variables|/Script/Engine.SoundBase|GetMaxDistance _soundcue))
    (else
      (return 0.0))))

(fn OnBurst(EitherVFX-preferthisversion)
  (params (MyTarget :type Actor%20Object%20Reference) (Parameters :type Gameplay%20Cue%20Parameters%20Structure%20%28by%20ref%29) (ParticleComponents :type Array%20of%20FXSystem%20Component%20Object%20References) (AudioComponents :type Array%20of%20Audio%20Component%20Object%20References) (BurstCameraShakeInstance :type Camera%20Shake%20Base%20Object%20Reference) (BurstDecalInstance :type Decal%20Actor%20Object%20Reference))
  (Fortnite|ShowSoundIndicatorLocalforTeam MyTarget MyTarget (Transformation|GetActorLocation MyTarget) (CallFunction|GetSoundIndicatorMaxDistance) "Chest" (Utilities|Array|MakeArray (Fortnite|GetFortPlayerControllerfromActor MyTarget)) (Team|GetActorTeam MyTarget) (Utilities|Array|MakeArray "Hostile" "Neutral" "Friendly") MyTarget "(R=0.000000,G=0.000000,B=0.000000,A=0.000000)" "/Game/UI/Foundation/Shell/Textures/Shell-Top-Crown.Shell-Top-Crown")
  (return))
